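# Sound Cue for the volcano lava lake ambience, mixing a steady lava bed with intermittent earth-movement rumbles.
# A mixer layers a looping lava lake wave, slightly boosted, with a loop that waits a random delay before playing one of four equally weighted earth-move waves.
# Overridden capsule attenuation with manual reverb send and limited concurrency that stops the quietest voice shape playback under an outdoor-ducking ambient sound class.
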SoundCue /Figment_S10/Athena/Sounds/Ambient/Volcano/Amb_Volcano_Lava_Lake_Cue
root=Mixer_0
nodes [
 {id=Delay_0, children [Random_0], properties {delayMax=15, delayMin=5}}
 {id=Looping_0, children [Delay_0]}
 {id=Mixer_0, children [WavePlayer_2,Looping_0], properties {inputVolume [1.25,1]}}
 {id=Random_0, children [WavePlayer_3,WavePlayer_4,WavePlayer_5,WavePlayer_6], properties {weights [1,1,1,1]}}
 {id=WavePlayer_2, properties {bLooping=true, soundWaveAssetPtr=&/Amb_Volcano_Lava_LakeLoop_01}}
 {id=WavePlayer_3, properties {soundWaveAssetPtr=&/Amb_Volcano_Earth_Move_04}}
 {id=WavePlayer_4, properties {soundWaveAssetPtr=/FNGameplayCues/Athena/Sounds/Ambient/Volcano/Amb_Volcano_Earth_Move_01}}
 {id=WavePlayer_5, properties {soundWaveAssetPtr=&/Amb_Volcano_Earth_Move_02}}
 {id=WavePlayer_6, properties {soundWaveAssetPtr=&/Amb_Volcano_Earth_Move_03}}
]
properties {
 bOverrideAttenuation=true
 bOverrideConcurrency=true
 maxDistance=6000
 soundClassObject=/Game/Sounds/FortSoundClassesAndModes/Fort_Ambient_DuckOutdoors_Athena_PSM
 volumeMultiplier=1.2
 attenuationOverrides {
  attenuationShape=Capsule
  falloffDistance=2500
  manualReverbSendLevel=15
  nonSpatializedRadiusStart=700
  reverbSendMethod=Manual
  stereoSpread=1000
  attenuationShapeExtents=<3500 1500>
 }
 concurrencyOverrides {maxCount=6, resolutionRule=StopQuietest}
}
