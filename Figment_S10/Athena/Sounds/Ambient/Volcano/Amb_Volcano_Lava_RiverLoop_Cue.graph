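# Ambient lava river loop cue for the Athena volcano, mixing two layers with separate attenuation overrides.
# One layer repeats random steam release one-shots after random delays, audible over the longer falloff.
# The other pairs a randomly chosen looping flow or churn bed with randomly delayed, repeating thick splashes.
# This closer layer has a shorter falloff, distance low-pass filtering and reverb send disabled.
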SoundCue /Figment_S10/Athena/Sounds/Ambient/Volcano/Amb_Volcano_Lava_RiverLoop_Cue
root=Mixer_0
nodes [
 {
  id=Attenuation_0
  children [Mixer_1]
  properties {bOverrideAttenuation=true, attenuationOverrides {falloffDistance=4000, stereoSpread=512, attenuationShapeExtents {x=0}}}
 }
 {
  id=Attenuation_1
  children [Mixer_2]
  properties {
   bOverrideAttenuation=true
   attenuationOverrides {
    bAttenuateWithLPF=true
    bEnableReverbSend=false
    falloffDistance=3000
    lPFFrequencyAtMax=3000
    lPFRadiusMax=2000
    lPFRadiusMin=1000
    nonSpatializedRadiusStart=512
    stereoSpread=750
    attenuationShapeExtents {x=512}
   }
  }
 }
 {id=Delay_0, children [Random_1], properties {delayMax=40, delayMin=20}}
 {id=Delay_1, children [Random_2], properties {delayMax=60, delayMin=20}}
 {id=Looping_0, children [Delay_0]}
 {id=Looping_1, children [Delay_1]}
 {id=Mixer_0, children [Attenuation_0, Attenuation_1], properties {inputVolume [1,1.5]}}
 {id=Mixer_1, children [Looping_0,null], properties {inputVolume [1.5,1]}}
 {id=Mixer_2, children [Random_3,Looping_1], properties {inputVolume [1,1]}}
 {id=Random_1, children [WavePlayer_1,WavePlayer_3,WavePlayer_4,WavePlayer_6,WavePlayer_7,WavePlayer_8], properties {weights [1,1,1,1,1,1]}}
 {id=Random_2, children [WavePlayer_9, WavePlayer_10, WavePlayer_11, WavePlayer_12], properties {weights [1,1,1,1]}}
 {id=Random_3, children [WavePlayer_0, WavePlayer_13], properties {weights [1,1]}}
 {
  id=WavePlayer_0
  properties {bLooping=true, soundWaveAssetPtr=&/AMB_Lava_Lake_Close_Heavy_Flow_Loop}
 }
 {id=WavePlayer_1, properties {soundWaveAssetPtr=&/AMB_Lava_Steam_Release_01}}
 {id=WavePlayer_10, properties {soundWaveAssetPtr=&/AMB_Lava_Splash_Thick_02}}
 {id=WavePlayer_11, properties {soundWaveAssetPtr=&/AMB_Lava_Splash_Thick_03}}
 {id=WavePlayer_12, properties {soundWaveAssetPtr=&/AMB_Lava_Splash_Thick_04}}
 {
  id=WavePlayer_13
  properties {bLooping=true, soundWaveAssetPtr=&/AMB_Lava_Lake_Close_Churn_Splash_Loop}
 }
 {id=WavePlayer_3, properties {soundWaveAssetPtr=&/AMB_Lava_Steam_Release_02}}
 {id=WavePlayer_4, properties {soundWaveAssetPtr=&/AMB_Lava_Steam_Release_03}}
 {id=WavePlayer_6, properties {soundWaveAssetPtr=&/AMB_Lava_Steam_Release_04}}
 {id=WavePlayer_7, properties {soundWaveAssetPtr=&/AMB_Lava_Steam_Release_05}}
 {id=WavePlayer_8, properties {soundWaveAssetPtr=&/AMB_Lava_Steam_Release_06}}
 {id=WavePlayer_9, properties {soundWaveAssetPtr=&/AMB_Lava_Splash_Thick_01}}
]
properties {
 maxDistance=4000
 soundClassObject=&Fort_Ambient_Athena
 volumeMultiplier=2
 attenuationOverrides {
  distanceAlgorithm=Custom
  falloffDistance=2000
  manualReverbSendLevel=15
  reverbSendMethod=Manual
  stereoSpread=512
  attenuationShapeExtents {x=1024}
  customAttenuationCurve {editorCurveData {keys [{arriveTangent=-2.9449, interpMode=RCIM_Cubic, leaveTangent=-2.9449, tangentMode=RCTM_User, time=0, value=2}, <at 1 0>]}}
 }
 concurrencyOverrides {maxCount=1, resolutionRule=PreventNew}
}
